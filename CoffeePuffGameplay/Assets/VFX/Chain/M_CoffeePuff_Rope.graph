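# Masked, two-sided, metallic lit material that renders the CoffeePuff chain rope from a chainlink texture.
# Two samples tile by object scale, offset half a tile and selected by vertex red and green, forming interlocking links.
# World position offset adds a parameterized bend plus time-animated spiral and return wiggles, sphere-masked around the middle of V.
# Emissive color fades along V, scales with time of day and spikes with the Flash parameter.
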Material /CoffeePuffGameplay/Assets/VFX/Chain/M_CoffeePuff_Rope
graph {
 properties {blendMode=BLEND_Masked, ditherOpacityMask=true, pixelDepthOffsetMode=PDOM_Legacy, twoSided=true, usedShadingModels=Default Lit}
 nodes [
  n1 = MaterialFunctionCall(Float2=TextureCoordinate()) {materialFunction=&BreakOutFloat2Components}
  n2 = Multiply(
   A=@n1.G
   B=Multiply(
    A=ScalarParameter() {defaultValue=0.75}
    B=MaterialFunctionCall() {materialFunction=&ObjectScale}.Scale X
   )
  )
  n3 = MaterialFunctionCall(R=@n1.R, G=@n2) {materialFunction=&AppendMany}
  n4 = Constant2Vector() {g=1, r=1}
  n5 = TextureSample(UVs=MaterialFunctionCall(XY=MaterialFunctionCall(UVs=@n3.RG, Texture Scale=@n4) {materialFunction=&ScaleUVsByCenter}.UVs) {materialFunction=&Swizzle}.YX) {
   samplerType=SAMPLERTYPE_Masks
   texture=/CoffeePuffGameplay/Athena/Effects/Particles/Skydiving/NS_Contrail_Motorcyclist_Female/Textures/T_Contrail_Motorcyclist_Female_Chainlink
  }
  n6 = VertexColor()
  n7 = MaterialFunctionCall(R=@n1.R, G=Add(A=@n2, B=Constant() {r=0.5})) {materialFunction=&AppendMany}
  n8 = Constant2Vector() {g=1, r=1}
  n9 = TextureSample(UVs=MaterialFunctionCall(XY=MaterialFunctionCall(UVs=@n7.RG, Texture Scale=@n8) {materialFunction=&ScaleUVsByCenter}.UVs) {materialFunction=&Swizzle}.YX) {
   samplerType=SAMPLERTYPE_Masks
   texture=&T_Contrail_Motorcyclist_Female_Chainlink
  }
  Add(A=Multiply(A=@n5.RGB, B=@n6.R), B=Multiply(A=@n9.RGB, B=@n6.G))
  n10 = MaterialFunctionCall() {materialFunction=&LinearGradient}
  n11 = ScalarParameter() {defaultValue=80, parameterName=BendBias}
  n12 = Saturate(None=Subtract(A=Multiply(A=@n10.VGradient), B=Power(Base=@n10.VGradient, Exp=@n11) {constExponent=20}))
  n13 = Multiply(A=ScalarParameter() {parameterName=Bend}, B=Constant3Vector() {constant=<rgba 0 1 0 0>})
  n14 = ScalarParameter() {defaultValue=1, parameterName=SpiralSine1}
  n15 = TextureCoordinate()
  n16 = Multiply(
   A=ComponentMask(None=Multiply(A=@n14, B=@n15)) {g=true}
   B=MaterialFunctionCall() {materialFunction=/Engine/Functions/Engine_MaterialFunctions02/WorldPositionOffset/ObjectScale}.Scale X
  )
  n17 = Time()
  n18 = Multiply(
   A=Sine(None=Add(A=@n16, B=@n17)) {period=10}
   B=Transform(None=Constant3Vector() {constant=<rgba 0 1.5 1.5 0>}) {transformSourceType=TRANSFORMSOURCE_Local}
  )
  n19 = ScalarParameter() {defaultValue=1, parameterName=SpiralSine2}
  n20 = TextureCoordinate()
  n21 = Multiply(
   A=ComponentMask(None=Multiply(A=@n19, B=@n20)) {g=true}
   B=MaterialFunctionCall() {materialFunction=&ObjectScale}.Scale X
  )
  n22 = Time()
  n23 = Multiply(
   A=Cosine(None=Add(A=@n21, B=@n22)) {period=10}
   B=Transform(None=Constant3Vector() {constant=<rgba 0 1.5 -1.5 0>}) {transformSourceType=TRANSFORMSOURCE_Local}
  )
  n24 = Multiply(
   A=ComponentMask(None=TextureCoordinate()) {g=true}
   B=MaterialFunctionCall() {materialFunction=&ObjectScale}.Scale X
  )
  n25 = Multiply(A=Time(), B=ScalarParameter() {defaultValue=2, parameterName=ReturnWiggleSpeed}) {constB=2}
  n26 = Multiply(
   A=Sine(None=Add(A=@n24, B=@n25)) {period=7}
   B=Transform(None=Constant3Vector() {constant=<rgba 0 0 3 0>}) {transformSourceType=TRANSFORMSOURCE_Local}
  )
  n27 = ScalarParameter() {parameterName=ReturnWiggle}
  n28 = Multiply(A=Add(A=Add(A=@n18, B=@n23), B=Multiply(A=@n26, B=@n27))) {constB=1.3}
  n29 = ScalarParameter() {defaultValue=15, parameterName=WiggleAmount}
  n30 = ComponentMask(None=TextureCoordinate()) {g=true}
  n31 = Constant() {r=0.5}
  n32 = ScalarParameter() {defaultValue=0.5, parameterName=SphereRadius}
  Add(
   A=Transform(None=Multiply(A=@n12, B=@n13)) {transformSourceType=TRANSFORMSOURCE_Local}
   B=Multiply(A=Multiply(A=@n28, B=@n29), B=SphereMask(A=@n30, B=@n31, Radius=@n32) {attenuationRadius=0.5, hardnessPercent=50})
  )
  n33 = ScalarParameter() {parameterName=Flash}
  n34 = MaterialFunctionCall(Float2=TextureCoordinate()) {materialFunction=&BreakOutFloat2Components}
  n35 = Constant() {r=0.1}
  n36 = VertexColor()
  n37 = Multiply(
   A=LinearInterpolate(
    A=VectorParameter() {parameterName=EmissiveColor, defaultValue=<rgba 1 0.1777 0.0114 1>}.RGB
    B=Constant3Vector() {constant=<rgba 0 0 0 1>}
    Alpha=Power(Base=@n34.G, Exp=@n35)
   )
   B=Add(A=Multiply(A=@n5.R, B=@n36.R), B=Multiply(A=@n5.G, B=@n36.G))
  )
  n38 = Constant() {r=0.5}
  n39 = Multiply(
   A=ScalarParameter() {defaultValue=5, parameterName=Emissive}
   B=MaterialFunctionCall(Value_Morning=@n38, Value_Day=Constant() {r=1}, Value_Evening=@n38, Value_Night=Constant() {r=0.3}) {materialFunction=&Scalar_Time_of_Day}.Result
  )
  MaterialFunctionCall(Emissive=Multiply(A=LinearInterpolate(Alpha=@n33) {constA=1, constB=100}, B=Multiply(A=@n37, B=@n39))) {materialFunction=&MF_FX_EmissiveAdjustments}
  n40 = Multiply(A=@n5.A, B=@n6.R)
  n41 = Multiply(A=@n6.G, B=@n9.A)
  Multiply(
   A=Power(Base=Add(A=@n40, B=@n41), Exp=ScalarParameter() {defaultValue=6, parameterName=ChainThickness}) {constExponent=1.8}
   B=ScalarParameter() {defaultValue=25, parameterName=ChainHardness}
  )
  ScalarParameter() {defaultValue=0.8, parameterName=Metallic}
  ScalarParameter() {defaultValue=1, parameterName=Specular}
  ScalarParameter() {defaultValue=0.15, parameterName=Roughness}
  VectorParameter() {parameterName=Base Color, defaultValue=<rgba 0.1094 0.1094 0.1094 0>}
 ]
}
